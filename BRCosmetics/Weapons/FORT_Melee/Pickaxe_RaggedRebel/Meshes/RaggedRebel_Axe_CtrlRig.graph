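# Control Rig for the Ragged Rebel pickaxe that spins the base_spin bone at speeds set by idle, sprint, melee and falling flags.
# Chained branches write a target speed and duration, with later states overriding earlier; sustained melee grows a swing timer that changes duration and blend, and resets when melee stops.
# A time loop latches those targets periodically, and an accumulated lerp smooths the speed driving a 360-degree loop on the bone's Y rotation.
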ControlRigBlueprint /BRCosmetics/Weapons/FORT_Melee/Pickaxe_RaggedRebel/Meshes/RaggedRebel_Axe_CtrlRig
graphs [
 ConstructionGraph {nodes [RigUnit_PrepareForExecution = &RigUnit_PrepareForExecution()]}
 RigVMFunctionLibrary {links [], nodes []}
 RigVMModel {nodes [
   RigUnit_BeginExecution = &RigUnit_BeginExecution()
   n1 = &RigVMFunction_ControlFlowBranch(ExecuteContext=@RigUnit_BeginExecution.ExecutePin, Condition=RigVMVariableNode() {variable=Is idle}.Value)
   n2 = &RigVMFunction_ControlFlowBranch(ExecuteContext=@n1.Completed, Condition=RigVMVariableNode() {variable=Is Sprinting}.Value)
   RigVMFunction_ControlFlowBranch = &RigVMFunction_ControlFlowBranch(ExecuteContext=@n2.Completed, Condition=RigVMVariableNode() {variable=Is Melee}.Value)
   n3 = &RigVMFunction_ControlFlowBranch(ExecuteContext=@RigVMFunction_ControlFlowBranch.Completed, Condition=RigVMVariableNode() {variable=Is Falling}.Value)
   RigVMFunction_Sequence = &RigVMFunction_Sequence(ExecuteContext=@n3.Completed)
   n4 = RigVMVariableNode() {variable=Initial Duration}
   n5 = /Script/RigVM.RigVMFunction_TimeLoop(Normalize=false, Speed=1, Duration=@n4.Value) {is=struct RigVMUnitNode}
   n6 = &RigVMFunction_ControlFlowBranch(
    Condition=/Script/RigVM.RigVMFunction_MathDoubleGreaterEqual(A=@n5.Relative, B=&RigVMFunction_MathDoubleSub(B=0.1, A=@n4.Value).Result) {is=struct RigVMUnitNode, template="GreaterEqual::Execute(in A,in B,out Result)"}.Result
    ExecuteContext=@RigVMFunction_Sequence.A
   )
   Multiply = &RigVMFunction_MathDoubleMul(B=0.5, A=@n4.Value)
   n7 = &RigVMFunction_ControlFlowBranch(
    Condition=&RigVMFunction_MathDoubleGreaterEqual(B=@Multiply.Result, A=@n5.Relative) {template="GreaterEqual::Execute(in A,in B,out Result)"}.Result
    ExecuteContext=@n6.Completed
   )
   AccumulateLerp = /Script/RigVM.RigVMFunction_AccumulateFloatLerp(
    bIntegrateDeltaTime=true
    TargetValue=RigVMVariableNode() {variable=Current Speed}.Value
    InitialValue=RigVMVariableNode() {variable=Next Speed}.Value
    Blend=RigVMVariableNode() {variable=blend time}.Value
   ) {is=struct RigVMUnitNode, template="AccumulateLerp::Execute(in TargetValue,in InitialValue,in Blend,in bIntegrateDeltaTime,out Result)"}
   RigVMFunction_TimeLoop = &RigVMFunction_TimeLoop(Duration=360, Normalize=false, Speed=@AccumulateLerp.Result)
   RigVMFunction_MathQuaternionFromEuler = &RigVMFunction_MathQuaternionFromEuler(Euler=(X=0,Y=0,Z=0), Euler.Y=@RigVMFunction_TimeLoop.Relative)
   &RigUnit_ModifyTransforms(
    ItemToModify=((Item=(Type=Bone,Name="base_spin"),Transform=(Rotation=(X=0,Y=0,Z=0,W=1),Translation=(X=0,Y=0,Z=0),Scale3D=(X=1,Y=1,Z=1))))
    Weight=1
    WeightMinimum=0
    ExecutePin=@n7.Completed
    ItemToModify.0.Transform.Rotation=@RigVMFunction_MathQuaternionFromEuler.Result
   )
   n8 = RigVMVariableNode() {variable=Speed And Duration}
   RigVMVariableNode(ExecuteContext=@n7.True, Value=@n8.Value.X) {variable=Next Speed}
   RigVMVariableNode(ExecuteContext=@n6.False, Value=RigVMVariableNode() {variable=Initial Duration}.Value) {variable=Initial Duration}
   RigVMVariableNode(ExecuteContext=@n6.True, Value=@n8.Value.Y) {variable=Initial Duration}
   > RigVMVariableNode(Value=@n8.Value.Y) {variable=Current Duration}
   > RigVMVariableNode(Value=@n8.Value.X) {variable=Current Speed}
   RigVMVariableNode(Value=(X=400,Y=0.1,Z=0), ExecuteContext=@n3.True) {variable=Speed And Duration}
   Subtract = &RigVMFunction_MathDoubleSub(B=3, A=RigVMVariableNode() {variable=swing duration}.Value)
   RigVMVariableNode(
    Value=RigVMDispatchNode(
     False=0
     Condition=/Script/RigVM.RigVMFunction_MathDoubleLessEqual(A=0, B=@Subtract.Result) {is=struct RigVMUnitNode, template="LessEqual::Execute(in A,in B,out Result)"}.Result
     True=@Subtract.Result
    ) {template="DISPATCH_RigVMDispatch_If(in Condition,in True,in False,out Result)"}.Result
    ExecuteContext=@RigVMFunction_ControlFlowBranch.False
   ) {variable=swing duration}
   Add = &RigVMFunction_MathDoubleAdd(B=0.03, A=RigVMVariableNode() {variable=swing duration}.Value)
   Clamp = &RigVMFunction_MathDoubleClamp(Maximum=3, Value=@Add.Result)
   RigVMVariableNode(Value=@Clamp.Result, ExecuteContext=@RigVMFunction_ControlFlowBranch.True) {variable=swing duration}
   GreaterEqual = &RigVMFunction_MathDoubleGreaterEqual(B=2.77, A=@Clamp.Result) {template="GreaterEqual::Execute(in A,in B,out Result)"}
   > RigVMVariableNode(
    Value=(X=2000,Y=1,Z=0)
    Value.Y=RigVMDispatchNode(False=0.5, True=@Clamp.Result, Condition=@GreaterEqual.Result) {template="DISPATCH_RigVMDispatch_If(in Condition,in True,in False,out Result)"}.Result
   ) {variable=Speed And Duration}
   n9 = &RigVMFunction_MathDoubleGreaterEqual(B=1.5, A=@Clamp.Result) {template="GreaterEqual::Execute(in A,in B,out Result)"}
   > RigVMVariableNode(Value=RigVMDispatchNode(
     True=1.5
     Condition=@GreaterEqual.Result
     False=RigVMDispatchNode(False=5, True=2, Condition=@n9.Result) {template="DISPATCH_RigVMDispatch_If(in Condition,in True,in False,out Result)"}.Result
    ) {template="DISPATCH_RigVMDispatch_If(in Condition,in True,in False,out Result)"}.Result) {variable=blend time}
   RigVMVariableNode(Value=(X=100,Y=0.1,Z=0), ExecuteContext=@n2.True) {variable=Speed And Duration}
   RigVMVariableNode(Value=(X=0,Y=0,Z=0), ExecuteContext=@n1.True) {variable=Speed And Duration}
   &RigVMFunction_MathDoubleSub(B=0.5, A=@n4.Value)
  ]}
]
hierarchy [
 Root {type=Connector, loc=<0 0 0>, rot=<0 0 0 1>}
 Root {type=Socket, loc=<0 0 0>, rot=<0 0 0 1>}
 b_AR_Root {loc=<0 0 0>, rot=<0 0 0.7071 0.7071>}
 Melee_TwoHanded_Handle {parent=b_AR_Root, loc=<8.1477 -0.008 -29.0793>, rot=<0 0 0.7071 -0.7071>}
 base_spin {parent=Melee_TwoHanded_Handle, loc=<0.7201 17.6853 80.5611>, rot=<0 0 1 0>}
 dyn_spin {parent=base_spin, loc=<0 0 0>, rot=<0 0 0 1>}
 FX_Idle {parent=b_AR_Root, type=Null, loc=<-8.6369 -0.8454 51.3081>, rot=<0 0 -0.7071 0.7071>}
 FX_trail_1 {parent=b_AR_Root, type=Null, loc=<-8.9894 -0.9761 54.8368>, rot=<0 0 0 1>}
 FX_trail_2 {parent=b_AR_Root, type=Null, loc=<-9.0046 -0.9761 47.6864>, rot=<0 0 0 1>}
]
variables [
 Current Duration {type=double, default=5}
 Current Speed {type=double, default=0}
 Initial Duration {type=double, default=1}
 Is Falling {type=bool, default=False}
 Is Melee {type=bool, default=False}
 Is Sprinting {type=bool, default=False}
 Is idle {type=bool, default=False}
 Isfrontend {type=bool, default=False}
 Next Speed {type=double, default=0}
 Speed And Duration {type=FVector, default=(X=0,Y=0,Z=0)}
 blend time {type=double, default=0.5}
 swing duration {type=double, default=0.2}
]
